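# Primary fire SoundCue for the Lotus Mustache weapon, built largely from Ashton Indigo LTM fire waves.
# It mixes a close modulated shot plus charge layer, randomized medium and distant variants, and a delayed close-range proximity grenade launcher reload-end tail.
# Each layer has its own distance attenuation, with close layers routed to the 3P fire sound class and others to Distant.
# Overridden concurrency caps instances per owner, stopping the oldest; the debug flag is left on.
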SoundCue /Game/Athena/Sounds/Lotus/Mustache/Lotus_Mustache_PrimaryFire
root=Mixer_1
nodes [
 {id=Attenuation_0, children [Mixer_3], properties {attenuationSettings=&Fort_Weapon_Projectile_Close_Athena_Att}}
 {
  id=Attenuation_1
  children [Random_4]
  properties {attenuationSettings=&Fort_Weapon_Projectile_Distant_Mono_Athena_Att}
 }
 {id=Attenuation_2, children [Delay_0], properties {attenuationSettings=&Fort_Weapon_Projectile_Close_Athena_Att}}
 {
  id=Attenuation_29
  children [Random_1]
  properties {attenuationSettings=&Fort_Weapon_Projectile_Medium_Mono_Athena_Att}
 }
 {id=Delay_0, children [WavePlayer_0], properties {delayMax=0.4, delayMin=0.4}}
 {id=Mixer_0, children [SoundClass_5, SoundClass_30, SoundClass_1], properties {inputVolume [0.9,0.8,0.6]}}
 {id=Mixer_1, children [Mixer_0,SoundClass_0], properties {inputVolume [1,0.5]}}
 {id=Mixer_3, children [Modulator_1,WavePlayer_3], properties {inputVolume [1,1]}}
 {id=Modulator_1, children [WavePlayer_1]}
 {
  id=Random_1
  children [WavePlayer_10, WavePlayer_11, WavePlayer_12, WavePlayer_13, WavePlayer_14, WavePlayer_15]
  properties {weights [1,1,1,1,1,1]}
 }
 {id=Random_4, children [WavePlayer_22, WavePlayer_23, WavePlayer_24, WavePlayer_25, WavePlayer_26], properties {weights [1,1,1,1,1]}}
 {
  id=SoundClass_0
  children [Attenuation_1]
  properties {soundClassOverride=&Fort_WeaponRanged_Fire_Distant}
 }
 {
  id=SoundClass_1
  children [Attenuation_2]
  properties {soundClassOverride=&Fort_WeaponRanged_Fire_3P_PSM}
 }
 {
  id=SoundClass_30
  children [Attenuation_29]
  properties {soundClassOverride=&Fort_WeaponRanged_Fire_Distant}
 }
 {
  id=SoundClass_5
  children [Attenuation_0]
  properties {soundClassOverride=&Fort_WeaponRanged_Fire_3P_PSM}
 }
 {
  id=WavePlayer_0
  properties {soundWaveAssetPtr=/Game/Sounds/Fort_Weapons/Launchers/Prox_GrenadeLauncher/Proximity_GrenadeLauncher_Reload_End_01}
 }
 {id=WavePlayer_1, properties {soundWaveAssetPtr=/Game/Athena/Sounds/LTMs/Ashton/Indigo/Ashton_Indigo_Primary_Fire_01_V3}}
 {id=WavePlayer_10, properties {soundWaveAssetPtr=/Game/Athena/Sounds/LTMs/Ashton/Indigo/Ashton_Indigo_Primary_Fire_Med_06}}
 {id=WavePlayer_11, properties {soundWaveAssetPtr=/Game/Athena/Sounds/LTMs/Ashton/Indigo/Ashton_Indigo_Primary_Fire_Med_01}}
 {id=WavePlayer_12, properties {soundWaveAssetPtr=/Game/Athena/Sounds/LTMs/Ashton/Indigo/Ashton_Indigo_Primary_Fire_Med_02}}
 {id=WavePlayer_13, properties {soundWaveAssetPtr=/Game/Athena/Sounds/LTMs/Ashton/Indigo/Ashton_Indigo_Primary_Fire_Med_03}}
 {id=WavePlayer_14, properties {soundWaveAssetPtr=/Game/Athena/Sounds/LTMs/Ashton/Indigo/Ashton_Indigo_Primary_Fire_Med_04}}
 {id=WavePlayer_15, properties {soundWaveAssetPtr=/Game/Athena/Sounds/LTMs/Ashton/Indigo/Ashton_Indigo_Primary_Fire_Med_05}}
 {id=WavePlayer_22, properties {soundWaveAssetPtr=/Game/Athena/Sounds/LTMs/Ashton/Indigo/Ashton_Indigo_Primary_Fire_Dist_01}}
 {id=WavePlayer_23, properties {soundWaveAssetPtr=/Game/Athena/Sounds/LTMs/Ashton/Indigo/Ashton_Indigo_Primary_Fire_Dist_02}}
 {id=WavePlayer_24, properties {soundWaveAssetPtr=/Game/Athena/Sounds/LTMs/Ashton/Indigo/Ashton_Indigo_Primary_Fire_Dist_03}}
 {id=WavePlayer_25, properties {soundWaveAssetPtr=/Game/Athena/Sounds/LTMs/Ashton/Indigo/Ashton_Indigo_Primary_Fire_Dist_04}}
 {id=WavePlayer_26, properties {soundWaveAssetPtr=/Game/Athena/Sounds/LTMs/Ashton/Indigo/Ashton_Indigo_Primary_Fire_Dist_05}}
 {id=WavePlayer_3, properties {soundWaveAssetPtr=/Game/Athena/Sounds/LTMs/Ashton/Indigo/Ashton_Indigo_Primary_Fire_Charge}}
]
properties {
 bDebug=true
 bOverrideConcurrency=true
 maxDistance=27500
 volumeMultiplier=0.7
 concurrencyOverrides {bLimitToOwner=true, maxCount=2, resolutionRule=StopOldest}
}
